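# Niagara system for the fiery hair and glowing eyes of the GoldAccomplishment male character cosmetic.
# All five emitters follow the skeletal mesh: burst-spawned mesh flames and glow sprites at bone offsets or sockets, plus rate-spawned embers from random sockets with drag and curl noise.
# User.VariantID selects glow and ember colours from arrays.
# Flames tint by time of day, and fire glow alpha flickers over loop age.
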NiagaraSystem /BRCosmetics/Characters/Player/Male/Medium/Bodies/M_MED_GoldAccomplishment/FX/NS_GoldAccomplishment_Hair
dependencies {
 usedDataInterfaces [
  &NiagaraDataInterfaceCurve
  &NiagaraDataInterfaceArrayColor
  &NiagaraDataInterfaceArrayInt32
  &NiagaraDataInterfaceSkeletalMesh
 ]
 usedDynamicInputs []
 usedModules [
  &self:SystemSpawnScript.SystemScriptSource.SystemScriptGraph.NiagaraNodeAssignment_0.SetVariables_#1
  &SystemState
  &self:SystemSpawnScript.SystemScriptSource.SystemScriptGraph.NiagaraNodeAssignment_2.SetVariables_#2
  &self:SystemSpawnScript.SystemScriptSource.SystemScriptGraph.NiagaraNodeAssignment_4.SetVariables_#3
 ]
}
emitters {
 EyeGlow {
  emitterUpdateScript [
   &EmitterState(Life Cycle Mode=::NewEnumerator1, Loop Behavior=::NewEnumerator1, Loop Duration Mode=::NewEnumerator1, UseLoopDelay=false)
   &SpawnBurst_Instantaneous(Spawn Count=2, Use Spawn Probability=false, UseLoopCountLimit=true)
  ]
  particleSpawnScript [
   &InitializeParticle(Lifetime=5, Sprite Size Mode=::NewEnumerator3, Uniform Sprite Size=8, UsePositionOffset=false, Write Material Random=true)
   &PartitionParticles()
  ]
  particleUpdateScript [
   &ParticleState(Kill Particles When Lifetime Has Elapsed=false, Let Infinitely Lived Particles Die When Emitter Deactivates=true)
   &SkeletalMeshLocation(
    Bone / Socket Index=&SelectIntFromArray(Direct Array Index=$Particles.Initial.PartitionParticles.Partition, Int Selection Array=DataInterface() {propertyValues {IntData [3,4]}})
    MaskSpawnGroup=false
    Mesh Sampling Type=::NewEnumerator5
    Skeletal Mesh=$System.Skeletal Mesh
    Socket Sampling Mode=::NewEnumerator5
   )
   &Color(Color=&SelectLinearColorFromArray(
     Array Sampling Mode=::NewEnumerator1
     Direct Array Index=&MakeIntFromFloat(Float=$User.VariantID)
     Color Selection Array=DataInterface() {propertyValues {ColorData [
        <rgba 1 0.2766 0 0.1>
        <rgba 1 0.8336 0 0.1>
        <rgba 0 0.9679 1 0.1>
        <rgba 1 0.8932 0.2061 0.1>
        <rgba 0.519 1 0.9602 0.1>
        <rgba 0.2449 0 1 0.1>
       ]}}
     Random Seed=?
    ))
  ]
  rendererClasses [&NiagaraSpriteRendererProperties]
  renderers [{rendererClass=&NiagaraSpriteRendererProperties}]
 }
 FireGlow {
  emitterUpdateScript [
   &EmitterState(Life Cycle Mode=::NewEnumerator1, Loop Behavior=::NewEnumerator1, Loop Duration Mode=::NewEnumerator1, UseLoopDelay=false)
   &SpawnBurst_Instantaneous(Spawn Count=15, Use Spawn Probability=false, UseLoopCountLimit=true)
  ]
  particleSpawnScript [&InitializeParticle(
    Lifetime=5
    Sprite Size Mode=::NewEnumerator3
    Uniform Sprite Size=&FloatFromCurve(
     FloatCurve=DataInterface() {propertyValues {Curve {keys [
         {interpMode=RCIM_Cubic, leaveTangent=0.6984, tangentMode=RCTM_Break, time=0, value=0.5}
         {interpMode=RCIM_Cubic, leaveTangent=-0.2297, tangentMode=RCTM_Break, time=0.5, value=0.9}
         <at 1 0.85>
        ]}}}
     CurveIndex=&ReturnNormalizedExecIndex(Normalized Index Includes One=false)
     Scale Curve=10
    )
    UsePositionOffset=false
    Write Material Random=true
   )]
  particleUpdateScript [
   &ParticleState(Kill Particles When Lifetime Has Elapsed=false, Let Infinitely Lived Particles Die When Emitter Deactivates=true)
   &SkeletalMeshLocation(
    Bone Sampling Mode=::NewEnumerator4
    MaskSpawnGroup=false
    Sampled Position Offset=&MakeVector(
     X=&FloatFromCurve(
      FloatCurve=DataInterface() {propertyValues {Curve {keys [{interpMode=RCIM_Cubic, leaveTangent=28.8262, tangentMode=RCTM_Break, time=0, value=12.5}, <at 1 21>]}}}
      CurveIndex=&ReturnNormalizedExecIndex(Normalized Index Includes One=false)
     )
     Y=&FloatFromCurve(
      FloatCurve=DataInterface() {propertyValues {Curve {keys [<at -0.0082 -7.1384>,<at 0.967 7.5294>]}}}
      CurveIndex=&ReturnNormalizedExecIndex(Normalized Index Includes One=false)
     )
     Z=0
    )
    Skeletal Mesh=$System.Skeletal Mesh
   )
   &Color(Color=&SelectLinearColorFromArray(
     Array Sampling Mode=::NewEnumerator1
     Direct Array Index=&MakeIntFromFloat(Float=$User.VariantID)
     Color Selection Array=DataInterface() {propertyValues {ColorData [
        <rgba 1 0.2766 0 0.1>
        <rgba 1 0.8336 0 0.1>
        <rgba 0 0.9679 1 0.1>
        <rgba 1 0.8932 0.2061 0.1>
        <rgba 0.519 1 0.9602 0.1>
        <rgba 0.2449 0 1 0.1>
       ]}}
     Random Seed=?
    ))
   &ScaleColor(
    Color Value To Scale=$Particles.Color
    Scale Alpha=&FloatFromCurve(
     FloatCurve=DataInterface() {propertyValues {Curve {keys [<at 0 1>,<at 0.1 0.1647>,<at 0.173 0.681>,<at 0.2977 0.2894>,<at 0.4303 0.9163>,<at 0.4901 0.2562>,<at 0.6201 0.1844>,<at 0.7787 0.9124>,<at 0.8671 0.112>,<at 1.0104 1>]}}}
     CurveIndex=$System.NormalizedLoopAge
     Scale Curve=&FN_Select_Float_from_Time_Of_Day(Morning=1, Day=1, Evening=1, Night=0.8)
    )
    ScaleA=true
    ScaleRGB=true
   )
  ]
  rendererClasses [&NiagaraSpriteRendererProperties]
  renderers [{rendererClass=&NiagaraSpriteRendererProperties}]
 }
 SideFlame {
  emitterUpdateScript [
   &EmitterState(Life Cycle Mode=::NewEnumerator1, Loop Behavior=::NewEnumerator1, Loop Duration Mode=::NewEnumerator1, UseLoopDelay=false)
   &SpawnBurst_Instantaneous(Use Spawn Probability=false, UseLoopCountLimit=true)
  ]
  particleSpawnScript [
   &InitializeParticle(Mesh Scale Mode=::NewEnumerator3, UsePositionOffset=false, Write Material Random=true)
   &InitialMeshOrientation(Mesh Orientation Mode=::NewEnumerator7, Rotation=<0 0.25 0>, Use Rotation Vector=true)
  ]
  particleUpdateScript [
   &ParticleState(Kill Particles When Lifetime Has Elapsed=false, Let Infinitely Lived Particles Die When Emitter Deactivates=true)
   &SkeletalMeshLocation(Bone Sampling Mode=::NewEnumerator4, MaskSpawnGroup=false, Sampled Position Offset=<-7.5911 -1.3605 -1>, Skeletal Mesh=$System.Skeletal Mesh)
   &Color(Color=&MultiplyLinearColor(
     Linear Color=$Particles.Initial.Color
     Scale Factor=&MakeVector4_FromLinearColor(LinearColor=&FNTimeOfDayColor(Night Color=<rgba 0.625 0.4674 0.6032 0.8>, Morning Color=<rgba 1 1 1 1>, Day Color=<rgba 1 1 1 1>, Evening Color=<rgba 1 1 1 1>))
    ))
   &ScaleColor(
    Scale Alpha=&FloatFromCurve(FloatCurve=DataInterface() {propertyValues {Curve {keys [{interpMode=RCIM_Cubic, leaveTangent=1.5973, tangentMode=RCTM_Break, time=0, value=0.2}, <at 1 0.95>]}}})
    ScaleA=true
    ScaleRGB=true
   )
  ]
  rendererClasses [&NiagaraMeshRendererProperties]
  renderers [{rendererClass=&NiagaraMeshRendererProperties, rendererIndex=0}]
 }
 SplineFlame {
  emitterUpdateScript [
   &EmitterState(Life Cycle Mode=::NewEnumerator1, Loop Behavior=::NewEnumerator1, Loop Duration Mode=::NewEnumerator1, UseLoopDelay=false)
   &SpawnBurst_Instantaneous(Use Spawn Probability=false, UseLoopCountLimit=true)
  ]
  particleSpawnScript [&InitializeParticle(Color=<rgba 0.5938 0.5938 0.5938 1>, Lifetime=5, UsePositionOffset=false, Write Material Random=true)]
  particleUpdateScript [
   &ParticleState(Kill Particles When Lifetime Has Elapsed=false, Let Infinitely Lived Particles Die When Emitter Deactivates=true)
   &SkeletalMeshLocation(Bone Sampling Mode=::NewEnumerator4, MaskSpawnGroup=false, Sampled Position Offset=<-0.1463 0 0.16>, Skeletal Mesh=$System.Skeletal Mesh)
   &Color(Color=&MultiplyLinearColor(
     Linear Color=$Particles.Initial.Color
     Scale Factor=&MakeVector4_FromLinearColor(LinearColor=&FNTimeOfDayColor(Night Color=<rgba 0.625 0.4674 0.6032 0.8>, Morning Color=<rgba 1 1 1 1>, Day Color=<rgba 1 1 1 1>, Evening Color=<rgba 1 1 1 1>))
    ))
   &DynamicMaterialParameters(
    Index 1 Param 3=14.8169
    Index 2 Param 1=0.15
    Index 3 Param 1=1
    Index 3 Param 2=$System.Age
    Index 3 Param 3=0.3
    Index 3 Param 4=$System.NormalizedLoopAge
    Module.Index1Param0Write=true
    Module.Index1Param1Write=true
    Module.Index1Param2Write=true
    Module.Index1Param3Write=true
    Module.Index2Param0Write=true
    Module.Index2Param1Write=true
    Module.Index2Param2Write=true
    Module.Index2Param3Write=true
    Module.Index3Param0Write=true
    Module.Index3Param1Write=true
    Module.Index3Param2Write=true
    Module.Index3Param3Write=true
    Write Parameter Index 0=false
    Write Parameter Index 1=true
    Write Parameter Index 2=true
    Write Parameter Index 3=true
   )
  ]
  rendererClasses [&NiagaraMeshRendererProperties]
  renderers [{rendererClass=&NiagaraMeshRendererProperties, rendererIndex=0}]
 }
 embers {
  emitterUpdateScript [
   &EmitterState()
   &SpawnRate(SpawnRate=5, Use Spawn Probability=false)
  ]
  particleSpawnScript [
   &InitializeParticle(
    Color=&SelectLinearColorFromArray(
     Array Sampling Mode=::NewEnumerator1
     Direct Array Index=&MakeIntFromFloat(Float=$User.VariantID)
     Color Selection Array=DataInterface() {propertyValues {ColorData [
        <rgba 2 0.3435 0 1>
        <rgba 2 1.6671 0 1>
        <rgba 0.0182 1.5991 2 1>
        <rgba 1.25 1.1165 0.2576 1>
        <rgba 1.25 0.541 1.0815 1>
        <rgba 0.2917 0.0374 1.2 1>
       ]}}
     Random Seed=?
    )
    Lifetime Max=0.5
    Lifetime Min=0.25
    Lifetime Mode=::NewEnumerator1
    Sprite Size Mode=::NewEnumerator1
    Uniform Sprite Size Max=1.1
    Uniform Sprite Size Min=0.45
    UsePositionOffset=false
    Write Material Random=true
   )
   &SkeletalMeshLocation(
    Bone / Socket Index=&RandomRangeInt(Minimum=0, Maximum=2)
    MaskSpawnGroup=false
    Mesh Sampling Type=::NewEnumerator5
    Skeletal Mesh=$System.Skeletal Mesh
    Socket Sampling Mode=::NewEnumerator5
   )
   &ShapeLocation(Hemisphere Distribution=<1 0.5>, MaskSpawnGroup=false, Override Seed=false, Sphere Radius=5, UseOwnerScale=true)
   &AddVelocity(Velocity=&RandomRangeVector(Minimum=<-5 20 10>, Maximum=<5 35 20>), Velocity Speed Scale=2)
  ]
  particleUpdateScript [
   &ParticleState()
   &ScaleSpriteSize(Uniform Curve Sprite Scale=DataInterface() {propertyValues {Curve {keys [
        {tangentMode=RCTM_User, time=0, value=1}
        {arriveTangent=-0.0405, interpMode=RCIM_Cubic, leaveTangent=-0.0405, tangentMode=RCTM_User, time=0.7009, value=1}
        {arriveTangent=-8.0089, interpMode=RCIM_Cubic, leaveTangent=-8.0089, tangentMode=RCTM_User, time=1, value=0}
       ]}}})
   &Drag(Drag=5.5, Use Linear Drag=true, Use Rotational Drag=true)
   &CurlNoiseForce(Noise Strength=50, Pan Noise=false)
   &SolveForcesAndVelocity(Clamp Velocity=false, Limit Acceleration=false)
   &ScaleSpriteSizeBySpeed(Max Scale Factor=<2 12>, Min Scale Factor=<1 2>)
  ]
  rendererClasses [&NiagaraSpriteRendererProperties]
  renderers [{rendererClass=&NiagaraSpriteRendererProperties}]
 }
}
summary {
 systemName=NS_GoldAccomplishment_Hair
 emitters [
  {emitterName=SideFlame, rendererClasses [&NiagaraMeshRendererProperties]}
  {emitterName=SplineFlame, rendererClasses [&NiagaraMeshRendererProperties]}
  {emitterName=FireGlow, rendererClasses [&NiagaraSpriteRendererProperties]}
  {emitterName=EyeGlow, rendererClasses [&NiagaraSpriteRendererProperties]}
  {emitterName=embers, rendererClasses [&NiagaraSpriteRendererProperties]}
 ]
 userVariables [
  User.bIsFrontEnd {type=NiagaraBool}
  User.bIsFrontEndPreview {type=NiagaraBool}
  User.VariantID {type=NiagaraFloat, defaultValue=0, description=""}
 ]
}
