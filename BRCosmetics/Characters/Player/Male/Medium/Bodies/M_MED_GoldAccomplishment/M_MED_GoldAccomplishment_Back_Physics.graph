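# Physics asset for the back skirt of the male medium GoldAccomplishment cosmetic body, defining its secondary-motion physics.
# Center, left and right chains each run from a pelvis-linked base through three capsule segments, with limited swing and locked twist.
# Capsules on the pelvis, both thighs and two large ones on the root provide body volumes around the skirt.
# Constraints linking pelvis to root, and thighs and skirt bases to pelvis, leave rotation free.
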PhysicsAsset /BRCosmetics/Characters/Player/Male/Medium/Bodies/M_MED_GoldAccomplishment/M_MED_GoldAccomplishment_Back_Physics
bodies {
 base_skirt_back_c [{shapeType=Capsule, length=0.505, radius=0.505, shapeName=base_skirt_back_c_capsule, rotation=<rot 0 -90 0>}]
 base_skirt_back_l [{shapeType=Capsule, length=0.505, radius=0.505, shapeName=base_skirt_back_l_capsule, rotation=<rot 0 -90 0>}]
 base_skirt_back_r [{shapeType=Capsule, length=0.505, radius=0.505, shapeName=base_skirt_back_r_capsule, rotation=<rot 0 -90 0>}]
 dyn_skirt_back_01_c [{shapeType=Capsule, length=6, radius=3, shapeName=dyn_skirt_back_01_c_capsule, center=<6 0 0>}]
 dyn_skirt_back_01_l [{shapeType=Capsule, length=6.2382, radius=2.7618, shapeName=dyn_skirt_back_01_l_capsule, center=<6 0 0>}]
 dyn_skirt_back_01_r [{shapeType=Capsule, length=7.8832, radius=1.1168, shapeName=dyn_skirt_back_01_r_capsule, center=<6 0 0>}]
 dyn_skirt_back_02_c [{shapeType=Capsule, length=9.7167, radius=3.2, shapeName=dyn_skirt_back_02_c_capsule, center=<8.7823 -1.4332 0>}]
 dyn_skirt_back_02_l [{shapeType=Capsule, length=6.86, radius=2.14, shapeName=dyn_skirt_back_02_l_capsule, center=<5.4295 -1.4332 0>}]
 dyn_skirt_back_02_r [{shapeType=Capsule, length=6, radius=3, shapeName=dyn_skirt_back_02_r_capsule, center=<5.4295 -1.4332 0>}]
 dyn_skirt_back_03_c [{shapeType=Capsule, length=11.6147, radius=4, shapeName=dyn_skirt_back_03_c_capsule, center=<6 -2.165 0>}]
 dyn_skirt_back_03_l [{shapeType=Capsule, length=6.79, radius=3.21, shapeName=dyn_skirt_back_03_l_capsule, center=<6 -2.165 0>}]
 dyn_skirt_back_03_r [{shapeType=Capsule, length=6, radius=4, shapeName=dyn_skirt_back_03_r_capsule, center=<6 -2.165 0>}]
 pelvis [{
   shapeType=Capsule
   length=11.1295
   radius=10.3639
   shapeName=pelvis_capsule
   center=<-0.526 -1.1052 -0.1024>
   rotation=<rot 4.6755 -90 -0.2276>
  }]
 root [
  {
   shapeType=Capsule
   length=132.4364
   radius=24.7823
   shapeName=root_capsule
   center=<13.8195 -43.12 -24.211>
   rotation=<rot 90 -172.875 97.1245>
  }
  {
   shapeType=Capsule
   length=132.4364
   radius=24.7823
   shapeName=root_capsule
   center=<-21.1193 -43.1203 -24.211>
   rotation=<rot 90 -165.9638 104.0357>
  }
 ]
 thigh_l [{
   shapeType=Capsule
   length=24.4077
   radius=9.4863
   shapeName=thigh_l_capsule
   center=<-13.7399 0 -2.605>
   rotation=<rot -90 -160.5288 160.5287>
  }]
 thigh_r [{shapeType=Capsule, length=24.972, radius=8.9219, shapeName=thigh_l_capsule, center=<13.7399 0 0>, rotation=<rot -90 -160.5288 -19.4713>}]
}
constraints [
 {
  bone1Name=thigh_l
  bone2Name=pelvis
  swing1LimitDegrees=45
  swing1Motion=Free
  swing2LimitDegrees=45
  swing2Motion=Free
  twistLimitDegrees=45
  twistMotion=Free
 }
 {
  bone1Name=thigh_r
  bone2Name=pelvis
  swing1LimitDegrees=45
  swing1Motion=Free
  swing2LimitDegrees=45
  swing2Motion=Free
  twistLimitDegrees=45
  twistMotion=Free
 }
 {
  bone1Name=base_skirt_back_c
  bone2Name=pelvis
  swing1LimitDegrees=45
  swing1Motion=Free
  swing2LimitDegrees=45
  swing2Motion=Free
  twistLimitDegrees=45
  twistMotion=Free
 }
 {
  bone1Name=dyn_skirt_back_01_c
  bone2Name=base_skirt_back_c
  swing1LimitDegrees=10
  swing2LimitDegrees=5
  twistLimitDegrees=45
  twistMotion=Locked
 }
 {
  bone1Name=dyn_skirt_back_02_c
  bone2Name=dyn_skirt_back_01_c
  swing1LimitDegrees=30
  swing2LimitDegrees=10
  twistLimitDegrees=10
  twistMotion=Locked
 }
 {
  bone1Name=dyn_skirt_back_03_c
  bone2Name=dyn_skirt_back_02_c
  swing1LimitDegrees=20
  swing2LimitDegrees=10
  twistLimitDegrees=20
  twistMotion=Locked
 }
 {
  bone1Name=base_skirt_back_l
  bone2Name=pelvis
  swing1LimitDegrees=45
  swing1Motion=Free
  swing2LimitDegrees=45
  swing2Motion=Free
  twistLimitDegrees=45
  twistMotion=Free
 }
 {
  bone1Name=dyn_skirt_back_01_l
  bone2Name=base_skirt_back_l
  swing1LimitDegrees=10
  swing2LimitDegrees=5
  twistLimitDegrees=45
  twistMotion=Locked
 }
 {
  bone1Name=dyn_skirt_back_02_l
  bone2Name=dyn_skirt_back_01_l
  swing1LimitDegrees=30
  swing2LimitDegrees=15
  twistLimitDegrees=45
  twistMotion=Locked
 }
 {
  bone1Name=dyn_skirt_back_03_l
  bone2Name=dyn_skirt_back_02_l
  swing1LimitDegrees=20
  swing2LimitDegrees=10
  twistLimitDegrees=45
  twistMotion=Locked
 }
 {
  bone1Name=base_skirt_back_r
  bone2Name=pelvis
  swing1LimitDegrees=45
  swing1Motion=Free
  swing2LimitDegrees=45
  swing2Motion=Free
  twistLimitDegrees=45
  twistMotion=Free
 }
 {
  bone1Name=dyn_skirt_back_01_r
  bone2Name=base_skirt_back_r
  swing1LimitDegrees=10
  swing2LimitDegrees=5
  twistLimitDegrees=45
  twistMotion=Locked
 }
 {
  bone1Name=dyn_skirt_back_02_r
  bone2Name=dyn_skirt_back_01_r
  swing1LimitDegrees=30
  swing2LimitDegrees=15
  twistLimitDegrees=45
  twistMotion=Locked
 }
 {
  bone1Name=dyn_skirt_back_03_r
  bone2Name=dyn_skirt_back_02_r
  swing1LimitDegrees=20
  swing2LimitDegrees=10
  twistLimitDegrees=45
  twistMotion=Locked
 }
 {
  bone1Name=pelvis
  bone2Name=root
  swing1LimitDegrees=45
  swing1Motion=Free
  swing2LimitDegrees=45
  swing2Motion=Free
  twistLimitDegrees=45
  twistMotion=Free
 }
]
