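# Physics asset for the front skirt of the medium male GoldAccomplishment player body cosmetic, with capsule bodies for the pelvis, thighs and skirt bones.
# Tiny base bodies link to the pelvis with free swing and twist, anchoring left, right and center skirt chains.
# Each side runs B_01 to B_02, then branches to B_03 and A, with limited swing and no twist; the center is a single limited segment.
# Swing limits differ between left and right.
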PhysicsAsset /BRCosmetics/Characters/Player/Male/Medium/Bodies/M_MED_GoldAccomplishment/M_MED_GoldAccomplishment_Front_Physics
bodies {
 base_skirt_front_B_l [{shapeType=Capsule, length=0.505, radius=0.505, shapeName=base_skirt_front_B_l_capsule, rotation=<rot 0 -90 0>}]
 base_skirt_front_B_r [{shapeType=Capsule, length=0.505, radius=0.505, shapeName=base_skirt_front_B_r_capsule, rotation=<rot 0 -90 0>}]
 base_skirt_front_c [{shapeType=Capsule, length=0.505, radius=0.505, shapeName=base_skirt_front_c_capsule, rotation=<rot 0 -90 0>}]
 dyn_skirt_front_A_l [{shapeType=Capsule, length=7.8378, radius=1.6879, shapeName=dyn_skirt_front_A_l_capsule, center=<7.343 0 0>, rotation=<rot 90 0 0>}]
 dyn_skirt_front_A_r [{shapeType=Capsule, length=7.8378, radius=1.6879, shapeName=dyn_skirt_front_A_r_capsule, center=<7.343 0 0>, rotation=<rot 90 0 0>}]
 dyn_skirt_front_B_01_l [{shapeType=Capsule, length=5, radius=2, center=<2 0.7282 0>, rotation=<rot 0 90.0001 -90>}]
 dyn_skirt_front_B_01_r [{shapeType=Capsule, length=5, radius=2, center=<2 0.7282 0>, rotation=<rot 0 90 0>}]
 dyn_skirt_front_B_02_l [{shapeType=Capsule, length=5, radius=2, shapeName=dyn_skirt_front_B_02_l_capsule, center=<2 0.7282 0>, rotation=<rot 0 90 0>}]
 dyn_skirt_front_B_02_r [{shapeType=Capsule, length=4.9588, radius=2.0412, shapeName=dyn_skirt_front_B_02_r_capsule, center=<2 1.0332 0>, rotation=<rot 0 90 0>}]
 dyn_skirt_front_B_03_l [{shapeType=Capsule, length=5, radius=2, shapeName=dyn_skirt_front_B_03_l_capsule, center=<2 0.7282 0>, rotation=<rot 0 90 0>}]
 dyn_skirt_front_B_03_r [{shapeType=Capsule, length=5.2145, radius=1.9255, shapeName=dyn_skirt_front_B_03_r_capsule, center=<1.8127 1.6883 0>, rotation=<rot 0 90 0>}]
 dyn_skirt_front_c [{shapeType=Capsule, length=7.5983, radius=1.9274, shapeName=dyn_skirt_front_c_capsule, center=<7.343 0 0>, rotation=<rot 90 0 0>}]
 pelvis [{shapeType=Capsule, length=12.049, radius=9.295}]
 thigh_l [
  {shapeType=Capsule, length=26.5392, radius=7.8524, center=<-16.6514 0.0225 -1.4513>, rotation=<rot 0.1444 -92.2643 -88.3498>}
  {
   shapeType=Capsule
   length=15.1451
   radius=3.2851
   shapeName=thigh_l_capsule
   center=<-12.0183 2.4727 -11.3945>
   rotation=<rot 90 180 -179.9999>
  }
 ]
 thigh_r [{shapeType=Capsule, length=22.6301, radius=9.9415, center=<17.7968 1.2636 1.0979>, rotation=<rot 0.239 -90.798 -87.2677>}]
}
constraints [
 {bone1Name=thigh_l, bone2Name=pelvis, swing1LimitDegrees=45, swing2LimitDegrees=45, twistLimitDegrees=45}
 {bone1Name=thigh_r, bone2Name=pelvis, swing1LimitDegrees=45, swing2LimitDegrees=45, twistLimitDegrees=45}
 {bone1Name=dyn_skirt_front_B_01_r, bone2Name=base_skirt_front_B_r, swing1LimitDegrees=20, swing2LimitDegrees=10, twistLimitDegrees=0}
 {bone1Name=dyn_skirt_front_B_01_l, bone2Name=base_skirt_front_B_l, swing1LimitDegrees=10, swing2LimitDegrees=10, twistLimitDegrees=0}
 {
  bone1Name=base_skirt_front_B_l
  bone2Name=pelvis
  swing1LimitDegrees=45
  swing1Motion=Free
  swing2LimitDegrees=45
  swing2Motion=Free
  twistLimitDegrees=45
  twistMotion=Free
 }
 {
  bone1Name=dyn_skirt_front_B_02_l
  bone2Name=dyn_skirt_front_B_01_l
  swing1LimitDegrees=10
  swing2LimitDegrees=5
  twistLimitDegrees=45
  twistMotion=Locked
 }
 {
  bone1Name=dyn_skirt_front_A_l
  bone2Name=dyn_skirt_front_B_02_l
  swing1LimitDegrees=10
  swing2LimitDegrees=10
  twistLimitDegrees=45
  twistMotion=Locked
 }
 {
  bone1Name=dyn_skirt_front_B_03_l
  bone2Name=dyn_skirt_front_B_02_l
  swing1LimitDegrees=10
  swing2LimitDegrees=5
  twistLimitDegrees=45
  twistMotion=Locked
 }
 {
  bone1Name=base_skirt_front_B_r
  bone2Name=pelvis
  swing1LimitDegrees=45
  swing1Motion=Free
  swing2LimitDegrees=45
  swing2Motion=Free
  twistLimitDegrees=45
  twistMotion=Free
 }
 {
  bone1Name=dyn_skirt_front_B_02_r
  bone2Name=dyn_skirt_front_B_01_r
  swing1LimitDegrees=10
  swing2LimitDegrees=5
  twistLimitDegrees=45
  twistMotion=Locked
 }
 {
  bone1Name=dyn_skirt_front_A_r
  bone2Name=dyn_skirt_front_B_02_r
  swing1LimitDegrees=20
  swing2LimitDegrees=15
  twistLimitDegrees=45
  twistMotion=Locked
 }
 {
  bone1Name=dyn_skirt_front_B_03_r
  bone2Name=dyn_skirt_front_B_02_r
  swing1LimitDegrees=25
  swing2LimitDegrees=5
  twistLimitDegrees=45
  twistMotion=Locked
 }
 {
  bone1Name=base_skirt_front_c
  bone2Name=pelvis
  swing1LimitDegrees=45
  swing1Motion=Free
  swing2LimitDegrees=45
  swing2Motion=Free
  twistLimitDegrees=45
  twistMotion=Free
 }
 {bone1Name=dyn_skirt_front_c, bone2Name=base_skirt_front_c, swing1LimitDegrees=25, swing2LimitDegrees=45, twistLimitDegrees=20}
]
