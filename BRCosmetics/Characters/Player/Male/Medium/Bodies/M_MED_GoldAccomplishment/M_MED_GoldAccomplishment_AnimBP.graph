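# Character-part AnimBP for the GoldAccomplishment medium male body, adding secondary motion to bandana, skirt and shoulder pads.
# The parent mesh's copied pose passes through Control Rig, bandana offsets and AnimDynamics, then skirt offsets and front/back RigidBody physics, then shoulder-pad constraints and offsets.
# The EventGraph turns movement states and reference-pose bone deltas into blend alphas for these corrections.
# A deprecated state function supplies the RigidBody gravity override.
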AnimBlueprint /BRCosmetics/Characters/Player/Male/Medium/Bodies/M_MED_GoldAccomplishment/M_MED_GoldAccomplishment_AnimBP
target_skeleton=/Game/Characters/Player/Male/Medium/Bodies/M_MED_GoldAccomplishment/M_MED_GoldAccomplishment_Skeleton
graphs [
 AnimGraph {nodes [
   n1 = "Constraint - base_shoulder_pad_l (PPP)"!(
    ComponentPose=Local To Component(LocalPose=Use cached pose 'RigidBodyCache'())
    Alpha=1
    ConstraintWeights_0=0.4
    ConstraintWeights_1:float=0.4
    ConstraintWeights_2:float=0.2
   ) -> (Pose) {
    alphaBoolBlend {blendOption=Linear}
    boneToModify {boneName=base_shoulder_pad_l}
    constraintSetup [{targetBone {boneName=upperarm_twist_01_l}}, {targetBone {boneName=upperarm_twist_02_l}}, {targetBone {boneName=deform_upperarm_up_l}}]
    constraintWeights [1,1,1]
   }
   n2 = "Constraint - base_shoulder_pad_r (PPP)"!(ComponentPose=@n1, Alpha=1, ConstraintWeights_0=0.4, ConstraintWeights_1:float=0.4, ConstraintWeights_2:float=0.2) -> (Pose) {
    alphaBoolBlend {blendOption=Linear}
    boneToModify {boneName=base_shoulder_pad_r}
    constraintSetup [{targetBone {boneName=upperarm_twist_01_r}}, {targetBone {boneName=upperarm_twist_02_r}}, {targetBone {boneName=deform_upperarm_up_r}}]
    constraintWeights [1,1,1]
   }
   n3 = "Transform (Modify) Bone - Bone: dyn_shoulder_pad_r"(
    ComponentPose=@n2
    Alpha=Get upperarm_twist_r() -> (upperarm_twist_r:double)
    Translation="-1,0,-0.2"
    Rotation="-5.066,2.854,0.090"
    Scale="1,1,1"
   ) {
    rotationMode=BMM_Additive
    rotationSpace=BCS_BoneSpace
    translationMode=BMM_Additive
    translationSpace=BCS_BoneSpace
    alphaBoolBlend {blendOption=Linear}
    boneToModify {boneName=dyn_shoulder_pad_r}
    rotation=<rot -5.066 2.8539 0.0901>
    translation {x=-1, z=-0.2}
   }
   n4 = "Transform (Modify) Bone - Bone: dyn_shoulder_pad_l"(ComponentPose=@n3, Alpha=Get upperarm_twist_l() -> (upperarm_twist_l:double), Translation="-1,0,-0.2", Rotation="0,8,0", Scale="1,1,1") {
    rotationMode=BMM_Additive
    rotationSpace=BCS_BoneSpace
    translationMode=BMM_Additive
    translationSpace=BCS_BoneSpace
    alphaBoolBlend {blendOption=Linear}
    boneToModify {boneName=dyn_shoulder_pad_l}
    rotation {yaw=8}
    translation {x=-1, z=-0.2}
   }
   n5 = "Transform (Modify) Bone - Bone: dyn_shoulder_pad_r"(ComponentPose=@n4, Alpha=Get upperarm_up_r() -> (upperarm_up_r:double), Translation="6,0,0", Rotation="8,0,0", Scale="1,1,1") {
    rotationMode=BMM_Additive
    rotationSpace=BCS_BoneSpace
    translationMode=BMM_Additive
    translationSpace=BCS_BoneSpace
    alphaBoolBlend {blendOption=Linear}
    boneToModify {boneName=dyn_shoulder_pad_r}
    rotation {pitch=8}
    translation {x=6}
   }
   n6 = "Transform (Modify) Bone - Bone: dyn_shoulder_pad_l"(ComponentPose=@n5, Alpha=Get upperarm_up_l() -> (upperarm_up_l:double), Translation="6,0,0", Rotation="8,0,0", Scale="1,1,1") {
    rotationMode=BMM_Additive
    rotationSpace=BCS_BoneSpace
    translationMode=BMM_Additive
    translationSpace=BCS_BoneSpace
    alphaBoolBlend {blendOption=Linear}
    boneToModify {boneName=dyn_shoulder_pad_l}
    rotation {pitch=8}
    translation {x=6}
   }
   Output Pose(Result=Component To Local(ComponentPose=@n6)) {layerGroup=None}
   n7 = Control Rig!(Source=Copy Pose From Mesh() {bUseAttachedParent=true}, Alpha, Is floating:bool=Get bIsSkyDiving(), Is Aiming:bool=Get bIsTargeting()) -> (Pose) {
    alphaBoolBlend {blendOption=Linear}
    controlRigAssetReference {blueprintRigClass=&/M_MED_GoldAc_Master_ControlRig.M_MED_GoldAc_Master_ControlRig_C}
   }
   n8 = "Transform (Modify) Bone - Bone: dyn_bandana_stretch_r"(
    ComponentPose=Local To Component(LocalPose=@n7)
    Alpha=Get HeadUp() -> (HeadUp:double)
    Translation="2.071,-6.459,2.000"
    Rotation="0,0,0"
    Scale="1,1,1"
   ) {
    rotationMode=BMM_Additive
    rotationSpace=BCS_BoneSpace
    scaleMode=BMM_Additive
    scaleSpace=BCS_BoneSpace
    translationMode=BMM_Additive
    translationSpace=BCS_BoneSpace
    alphaBoolBlend {blendOption=Linear, blendOutTime=2}
    boneToModify {boneName=dyn_bandana_stretch_r}
    translation=<2.071 -6.459 2>
   }
   n9 = Get HeadUp_Floating() -> (HeadUp_Floating:double)
   n10 = "Transform (Modify) Bone - Bone: dyn_bandana_stretch_l"(ComponentPose=@n8, Alpha=@n9, Translation="4,4,1.5", Rotation="0,0,0", Scale="1,1,1") {
    rotationMode=BMM_Additive
    rotationSpace=BCS_BoneSpace
    scaleMode=BMM_Additive
    scaleSpace=BCS_BoneSpace
    translationMode=BMM_Additive
    translationSpace=BCS_BoneSpace
    alphaBoolBlend {blendOption=Linear, blendOutTime=2}
    boneToModify {boneName=dyn_bandana_stretch_l}
    translation=<4 4 1.5>
   }
   n11 = "Transform (Modify) Bone - Bone: dyn_bandana_stretch_r"(ComponentPose=@n10, Alpha=@n9, Translation="-5,6,-2", Rotation="0,0,0", Scale="1,1,1") {
    rotationMode=BMM_Additive
    rotationSpace=BCS_BoneSpace
    scaleMode=BMM_Additive
    scaleSpace=BCS_BoneSpace
    translationMode=BMM_Additive
    translationSpace=BCS_BoneSpace
    alphaBoolBlend {blendOption=Linear, blendOutTime=2}
    boneToModify {boneName=dyn_bandana_stretch_r}
    translation=<-5 6 -2>
   }
   n12 = Get Is Parachute Opened() -> (Is Parachute Opened:double)
   n13 = "Transform (Modify) Bone - Bone: dyn_bandana_stretch_r"(ComponentPose=@n11, Alpha=@n12, Translation="0,0,0", Rotation="0,0,0", Scale="3,1,1") {
    rotationMode=BMM_Additive
    rotationSpace=BCS_BoneSpace
    scaleMode=BMM_Additive
    scaleSpace=BCS_BoneSpace
    translationMode=BMM_Additive
    translationSpace=BCS_BoneSpace
    alphaBoolBlend {blendOption=Linear, blendOutTime=2}
    alphaScaleBiasClamp {bInterpResult=true}
    boneToModify {boneName=dyn_bandana_stretch_r}
    scale {x=3}
   }
   n14 = "Transform (Modify) Bone - Bone: dyn_bandana_stretch_l"(ComponentPose=@n13, Alpha=@n12, Translation="0,0,0", Rotation="0,0,0", Scale="3,1,1") {
    rotationMode=BMM_Additive
    rotationSpace=BCS_BoneSpace
    scaleMode=BMM_Additive
    scaleSpace=BCS_BoneSpace
    translationMode=BMM_Additive
    translationSpace=BCS_BoneSpace
    alphaBoolBlend {blendOption=Linear, blendOutTime=2}
    alphaScaleBiasClamp {bInterpResult=true}
    boneToModify {boneName=dyn_bandana_stretch_l}
    scale {x=3}
   }
   n15 = Get Is Sky Diving() -> (Is Sky Diving:double)
   n16 = "Transform (Modify) Bone - Bone: bandana_l"(ComponentPose=@n14, Alpha=@n15, Translation="3.478,0.402,3.272", Rotation="0,0,0", Scale="1,1,1") {
    rotationMode=BMM_Additive
    rotationSpace=BCS_BoneSpace
    translationMode=BMM_Additive
    translationSpace=BCS_BoneSpace
    alphaBoolBlend {blendInTime=5, blendOption=Linear, blendOutTime=5}
    alphaScaleBiasClamp {bInterpResult=true}
    boneToModify {boneName=bandana_l}
    translation=<3.478 0.402 3.272>
   }
   n17 = "Transform (Modify) Bone - Bone: bandana_r"(ComponentPose=@n16, Alpha=@n15, Translation="-2,0,3", Rotation="0,10,0", Scale="1,1,1") {
    rotationMode=BMM_Additive
    rotationSpace=BCS_BoneSpace
    translationMode=BMM_Additive
    translationSpace=BCS_BoneSpace
    alphaBoolBlend {blendInTime=5, blendOption=Linear, blendOutTime=5}
    alphaScaleBiasClamp {bInterpResult=true}
    boneToModify {boneName=bandana_r}
    rotation {yaw=10}
    translation {x=-2, z=3}
   }
   n18 = Get IsFloating() -> (IsFloating:double)
   n19 = "Constraint - bandana_l (P)"(ComponentPose=@n17, Alpha=@n18, ConstraintWeights_0=1) {
    alphaBoolBlend {blendOption=Linear}
    alphaScaleBiasClamp {bInterpResult=true, bMapRange=true, interpSpeedDecreasing=3, interpSpeedIncreasing=3, inRange {max=0, min=1}, outRange {max=0.7}}
    boneToModify {boneName=bandana_l}
    constraintSetup [{targetBone {boneName=clavicle_l}}]
    constraintWeights [1]
   }
   n20 = "Constraint - bandana_r (P)"(ComponentPose=@n19, Alpha=@n18, ConstraintWeights_0=1) {
    alphaBoolBlend {blendOption=Linear}
    alphaScaleBiasClamp {bInterpResult=true, bMapRange=true, interpSpeedDecreasing=3, interpSpeedIncreasing=3, inRange {max=0, min=1}, outRange {max=0.7}}
    boneToModify {boneName=bandana_r}
    constraintSetup [{targetBone {boneName=clavicle_r}}]
    constraintWeights [1]
   }
   n21 = Anim Dynamics - Bone: dyn_bandana(ComponentPose=@n20, Alpha=1, ExternalForce="0,0,0") {
    angularDampingOverride=1
    angularSpringConstant=20000
    bAngularSpring=true
    bOverrideAngularDamping=true
    bUseGravityOverride=true
    gravityScale=0.35
    simulationSpace=RootRelative
    alphaBoolBlend {blendOption=Linear}
    boundBone {boneName=dyn_bandana}
    componentLinearAccScale=<1 1 1>
    componentLinearVelScale=<1 1 1>
    gravityOverride {z=-980}
    physicsBodyDefinitions [{
      boundBone {boneName=dyn_bandana}
      boxExtents=<40 40 40>
      constraintSetup {angularLimitsMax=<0 0 0>, angularLimitsMin=<0 -20 0>, angularTarget=<1 0 0>}
      localJointOffset=<2 0 0>
     }]
   }
   Save cached pose 'ScaftCache'(Pose=Component To Local(ComponentPose=@n21)) {cacheName=ScaftCache}
   n22 = Get IsFloating()
   n23 = "Transform (Modify) Bone - Bone: base_skirt_front_B_r"(
    ComponentPose=Local To Component(LocalPose=Use cached pose 'ScaftCache'())
    Alpha=@n22
    Translation="0,0,0"
    Rotation="-10,10,-0.007"
    Scale="1,1,1"
   ) {
    rotationMode=BMM_Additive
    rotationSpace=BCS_BoneSpace
    alphaBoolBlend {blendOption=Linear, blendOutTime=2}
    boneToModify {boneName=base_skirt_front_B_r}
    rotation=<rot -10 10 -0.007>
   }
   n24 = "Transform (Modify) Bone - Bone: base_skirt_front_B_l"(ComponentPose=@n23, Alpha=@n22, Translation="0,0,0", Rotation="10,10,0.058", Scale="1,1,1") {
    rotationMode=BMM_Additive
    rotationSpace=BCS_BoneSpace
    alphaBoolBlend {blendOption=Linear, blendOutTime=2}
    boneToModify {boneName=base_skirt_front_B_l}
    rotation=<rot 10 10 0.058>
   }
   n25 = Get Is crouching() -> (Is crouching:double)
   n26 = "Transform (Modify) Bone - Bone: base_skirt_back_c"(ComponentPose=@n24, Alpha=@n25, Translation="0,0,0", Rotation="10,0,0", Scale="1,1,1") {
    rotationMode=BMM_Additive
    rotationSpace=BCS_BoneSpace
    alphaBoolBlend {blendOption=Linear, blendOutTime=2}
    alphaScaleBiasClamp {bInterpResult=true}
    boneToModify {boneName=base_skirt_back_c}
    rotation {pitch=10}
   }
   n27 = "Transform (Modify) Bone - Bone: base_skirt_back_r"(ComponentPose=@n26, Alpha=@n25, Translation="0,0,0", Rotation="5,0,0", Scale="1,1,1") {
    rotationMode=BMM_Additive
    rotationSpace=BCS_BoneSpace
    alphaBoolBlend {blendOption=Linear, blendOutTime=2}
    alphaScaleBiasClamp {bInterpResult=true}
    boneToModify {boneName=base_skirt_back_r}
    rotation {pitch=5}
   }
   n28 = "Transform (Modify) Bone - Bone: base_skirt_back_l"(ComponentPose=@n27, Alpha=@n25, Translation="0.046,-1,0.02", Rotation="5,10,0", Scale="1,1,1") {
    rotationMode=BMM_Additive
    rotationSpace=BCS_BoneSpace
    translationMode=BMM_Additive
    translationSpace=BCS_BoneSpace
    alphaBoolBlend {blendOption=Linear, blendOutTime=2}
    alphaScaleBiasClamp {bInterpResult=true}
    boneToModify {boneName=base_skirt_back_l}
    rotation {pitch=5, yaw=10}
    translation=<0.046 -1 0.02>
   }
   n29 = Get Walking Crouching() -> (Walking Crouching:double)
   n30 = "Transform (Modify) Bone - Bone: base_skirt_front_c"(ComponentPose=@n28, Alpha=@n29, Translation="-3,0,0", Rotation="0,0,0", Scale="1,1,1") {
    rotationMode=BMM_Additive
    rotationSpace=BCS_BoneSpace
    translationMode=BMM_Additive
    translationSpace=BCS_BoneSpace
    alphaBoolBlend {blendOption=Linear, blendOutTime=2}
    boneToModify {boneName=base_skirt_front_c}
    translation {x=-3}
   }
   n31 = "Transform (Modify) Bone - Bone: base_skirt_front_B_l"(ComponentPose=@n30, Alpha=@n29, Translation="0.427,-0.898,-4.198", Rotation="0.031,0.096,-10.007", Scale="1,1,1") {
    rotationMode=BMM_Additive
    rotationSpace=BCS_BoneSpace
    translationMode=BMM_Additive
    translationSpace=BCS_BoneSpace
    alphaBoolBlend {blendOption=Linear, blendOutTime=2}
    boneToModify {boneName=base_skirt_front_B_l}
    rotation=<rot 0.031 0.0961 -10.007>
    translation=<0.427 -0.898 -4.198>
   }
   n32 = "Transform (Modify) Bone - Bone: base_skirt_front_B_r"(ComponentPose=@n31, Alpha=@n29, Translation="0,-0.5,0", Rotation="-7.580,7.530,8.731", Scale="1,1,1") {
    rotationMode=BMM_Additive
    rotationSpace=BCS_BoneSpace
    translationMode=BMM_Additive
    translationSpace=BCS_BoneSpace
    alphaBoolBlend {blendOption=Linear, blendOutTime=2}
    boneToModify {boneName=base_skirt_front_B_r}
    rotation=<rot -7.58 7.53 8.731>
    translation {y=-0.5}
   }
   n33 = "Transform (Modify) Bone - Bone: base_skirt_back_c"(ComponentPose=@n32, Alpha=1, Translation="0,0,0", Rotation="0,20,0", Scale="1,1,1") {
    rotationMode=BMM_Additive
    rotationSpace=BCS_BoneSpace
    translationMode=BMM_Additive
    translationSpace=BCS_BoneSpace
    alphaBoolBlend {blendOption=Linear, blendOutTime=2}
    boneToModify {boneName=base_skirt_back_c}
    rotation {yaw=20}
   }
   n34 = Get GravityOverride()
   n35 = RigidBody!(ComponentPose=@n33, Alpha, OverrideWorldGravity:Vector=@n34, ExternalForce) -> (Pose) {
    bClampLinearTranslationLimitToRefPose=true
    bOverrideWorldGravity=true
    overridePhysicsAsset=&/M_MED_GoldAccomplishment_Front_Physics
    simulationSpace=BaseBoneSpace
    alphaBoolBlend {blendOption=Linear}
    baseBoneRef {boneName=pelvis}
    componentLinearAccScale=<0.3 0.3 0.3>
    componentLinearVelScale=<0.2 0.3 0.3>
   }
   n36 = RigidBody(ComponentPose=@n35, OverrideWorldGravity=@n34) {
    bClampLinearTranslationLimitToRefPose=true
    bOverrideWorldGravity=true
    overridePhysicsAsset=&/M_MED_GoldAccomplishment_Back_Physics
    simulationSpace=BaseBoneSpace
    alphaBoolBlend {blendOption=Linear}
    baseBoneRef {boneName=pelvis}
    componentLinearAccScale=<0.3 0.5 0.3>
    componentLinearVelScale=<0.3 0.5 0.3>
   }
   Save cached pose 'RigidBodyCache'(Pose=Component To Local(ComponentPose=@n36)) {cacheName=RigidBodyCache}
  ]}
 EventGraph {nodes [
   Event BlueprintUpdateAnimation()
   > n1 = Sequence!(execute) -> (then_0, then_1, then_2:Exec, then_3:Exec, then_4:Exec)
   Branch(execute=@n1.then_3, Condition=Get bIsTargeting())
   > Set AimHeadUp(AimHeadUp:double=Get HeadUp()) -> (Output_Get:double)
   n2 = Get bIsMoving2D()
   n3 = Get bIsCrouching()
   Set Walking Crouching(execute=@n1.then_4, Walking Crouching:double="To Float (Boolean)"(InBool=AND Boolean(A=@n2, B=@n3))) -> (Output_Get:double)
   > Set Is Parachute Opened(Is Parachute Opened:double="To Float (Boolean)"(InBool=Get bIsParachuteOpened())) -> (Output_Get:double)
   > Set Is crouching(Is crouching:double="To Float (Boolean)"(InBool=Get bIsCrouching())) -> (Output_Get:double)
   n4 = Get bIsSkydiveFloating()
   n5 = Get bIsParachuteOpened()
   Set IsFloating(execute=@n1.then_1, IsFloating:double="To Float (Boolean)"(InBool=OR Boolean(A=@n5, B=@n4))) -> (Output_Get:double)
   > Set Is Sky Diving(Is Sky Diving:double="To Float (Boolean)"(InBool=Get bIsSkyDiving())) -> (Output_Get:double)
   n6 = Get Skel Mesh()
   n7 = float * float(
    A=GetAlphaFromReferenceTransformDelta(SkeletalMeshComponent=@n6, BoneName=spine_05, BaseName=head, InputRange=(X=0,Y=75), TwistAxis="0,0,1")
    B=1
   ) {is=title K2Node_PromotableOperator}
   Set HeadUp(execute=@n1.then_2, HeadUp:double=@n7) -> (Output_Get:double)
   > Set HeadUp_Floating(HeadUp_Floating:double=float * float(A=@n7, B=Get IsFloating())) -> (Output_Get:double)
   n8 = GetAlphaFromReferenceTransformDelta(
    SkeletalMeshComponent=@n6
    BoneName=upperarm_r
    BaseName=upperarm_twist_02_r
    DeltaComponentType=EulerX
    InputRange=(X=-30,Y=30)
    OutputRange=(X=-1,Y=1)
    TwistAxis="0,0,1"
   )
   > Set upperarm_twist_r(upperarm_twist_r:double=float + float(
     A=Remap Float(Input=@n8, OldRange=(X=0,Y=-1), NewRange=(X=0,Y=1.5), Clamp=true) -> (Output=~)
     B=Remap Float(Input=@n8, OldRange=(X=0,Y=1), NewRange=(X=0,Y=0), Clamp=true) -> (Output=~)
    )) -> (Output_Get:double)
   n9 = GetAlphaFromReferenceTransformDelta(
    SkeletalMeshComponent=@n6
    BoneName=upperarm_l
    BaseName=upperarm_twist_02_l
    DeltaComponentType=EulerX
    InputRange=(X=-30,Y=30)
    OutputRange=(X=-1,Y=1)
    TwistAxis="0,0,1"
   )
   > Set upperarm_twist_l(upperarm_twist_l:double=float + float(
     A=Remap Float(Input=@n9, OldRange=(X=0,Y=-1), NewRange=(X=0,Y=1.5), Clamp=true) -> (Output=~)
     B=Remap Float(Input=@n9, OldRange=(X=0,Y=1), NewRange=(X=0,Y=0), Clamp=true) -> (Output=~)
    )) -> (Output_Get:double)
   > Set upperarm_up_r(upperarm_up_r:double=GetAlphaFromReferenceTransformDelta(SkeletalMeshComponent=@n6, BoneName=upperarm_r, BaseName=clavicle_r, DeltaComponentType=EulerY, InputRange=(X=0,Y=90), TwistAxis="0,0,1")) -> (Output_Get:double)
   > Set upperarm_up_l(upperarm_up_l:double=GetAlphaFromReferenceTransformDelta(SkeletalMeshComponent=@n6, BoneName=upperarm_l, BaseName=clavicle_l, DeltaComponentType=EulerY, InputRange=(X=0,Y=90), TwistAxis="0,0,1")) -> (Output_Get:double)
   n10 = Gravity Override State Function DEPRECATED(
    execute=@n1.then_0
    CharacterPartAnimInstance=Self-Reference()
    GameTimeInSeconds=GetGameTimeInSeconds()
    GravityOverrideParams=Get Gravity Override Params()
   )
   > Set GravityOverride(GravityOverride=@n10)
   Event BlueprintInitializeAnimation()
   > Set Skel Mesh(Skel Mesh=GetOwnerSkeletalMeshComponent())
   Get bIsSkyDiving()
  ]}
]
variables [
 AimHeadUp {type=number}
 Gravity Override Params {
  type=GravityOverrideParamsStruct
  default {skydive_FloatingDown {gravityOverride=<0 -700 100>, windAmplitude=<200 -500 200>, windFrequency=<1.5 7 800>}}
 }
 GravityOverride {type=Vector}
 HeadUp {type=number}
 HeadUp_Floating {type=number}
 Is crouching {type=number}
 Is Parachute Opened {type=number}
 Is Sky Diving {type=number}
 IsFloating {type=number}
 Skel Mesh {type=SkeletalMeshComponent}
 upperarm_twist_l {type=number}
 upperarm_twist_r {type=number}
 upperarm_up_l {type=number}
 upperarm_up_r {type=number}
 Walking Crouching {type=number}
]
