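# Post-process AnimBP for the GoldAccomplishment medium male face accessory, which damps the accessory's head rotation outside of emotes.
# It copies the pose from the attached parent mesh and caches it.
# Each update reads the head's delta from ref pose relative to neck_02, negates it and scales each axis by a percentage.
# That counter-rotation is added to the head bone in bone space, while emotes use the unmodified copied pose.
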AnimBlueprint /BRCosmetics/Characters/Player/Male/Medium/Bodies/M_MED_GoldAccomplishment/Parts/M_MED_GoldAccomplishment_FaceAcc_PostProcess_AnimBP
target_skeleton=&/M_MED_GoldAccomplishment_FaceAcc_Skeleton
graphs [
 AnimGraph {nodes [
   n1 = "Transform (Modify) Bone - Bone: head"(
    ComponentPose=Local To Component(LocalPose=Use cached pose 'Start_CACHE'())
    Alpha=1
    Translation="0,0,0"
    Rotation=Get head rot() -> (head rot:Rotator)
    Scale="1,1,1"
   ) {
    rotationMode=BMM_Additive
    rotationSpace=BCS_BoneSpace
    alphaBoolBlend {blendOption=Linear}
    boneToModify {boneName=head}
    rotation=<rot 50 50 50>
   }
   Output Pose(Result=Blend Poses by bool(BlendPose_0=Use cached pose 'Start_CACHE'(), BlendPose_1=Component To Local(ComponentPose=@n1), bActiveValue=Get bIsPlayingEmote())) {layerGroup=None}
   Save cached pose 'Start_CACHE'(Pose=Copy Pose From Mesh() {bUseAttachedParent=true}) {cacheName=Start_CACHE}
  ]}
 EventGraph {nodes [
   Event BlueprintInitializeAnimation()
   > Set Skel Mesh(Skel Mesh=GetOwnerSkeletalMeshComponent())
   Event BlueprintUpdateAnimation()
   n1 = GetOwnerSkeletalMeshComponent()
   n2 = BreakRotator(InRot=BreakTransform(InTransform=GetDeltaTransformFromRefPose(self=@n1, BoneName=head, BaseName=neck_02)).Rotation)
   n3 = Get Y percentage() -> (Y percentage:float)
   n4 = Get X percentage() -> (X percentage:float)
   n5 = Get Z percentage() -> (Z percentage:float)
   > Set head rot(head rot:Rotator=MakeRotator(
     Roll=float * float!(A=@n2, B=-1, C:double=@n4) -> (ReturnValue)
     Pitch=float * float(A=@n2.Pitch, B=-1, C=@n3)
     Yaw=float * float(A=@n2.Yaw, B=-1, C=@n5)
    )) -> (Output_Get:Rotator)
  ]}
]
variables [
 head rot {type=Rotator}
 Skel Mesh {type=SkeletalMeshComponent}
 X percentage {type=number, default=0.8}
 Y percentage {type=number, default=0.3}
 Z percentage {type=number, default=0.3}
]
